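# Control Rig for the TangyRadishFlameLong female character's face accessory that correctively offsets long-hair base bones as the head moves.
# A spherical pose reader on the head bone outputs the weight for a Modify Transforms node acting on those hair bones.
# One item's offset switches on skydiving; otherwise its X translation depends on PlayEmote, then isTargeting, through nested If nodes.
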ControlRigBlueprint /BRCosmetics/Characters/Player/Female/Medium/Bodies/F_MED_TangyRadishFlameLong/Meshes/Parts/F_MED_TangyRadishFlameLong_FaceAcc_CtrlRig
graphs [
 RigVMFunctionLibrary {links [], nodes []}
 RigVMModel {nodes [
   RigUnit_BeginExecution = &RigUnit_BeginExecution()
   RigVMFunction_Sequence = &RigVMFunction_Sequence(ExecuteContext=@RigUnit_BeginExecution.ExecutePin)
   n1 = &RigUnit_SphericalPoseReader(
    ActiveRegionScaleFactors=(PositiveWidth=1,NegativeWidth=0.2,PositiveHeight=1.5,NegativeHeight=1.5)
    ActiveRegionSize=0.1
    Debug=(bDrawDebug=false,bDraw2D=false,bDrawLocalAxes=false,DebugScale=25,DebugSegments=20,DebugThickness=0.25)
    DriverAxis=(X=0,Y=1,Z=0)
    DriverItem=(Type=Bone,Name="head")
    FalloffRegionScaleFactors=(PositiveWidth=3,NegativeWidth=1.8,PositiveHeight=2.5,NegativeHeight=2.5)
    FalloffSize=0.2
    FlipHeightScaling=False
    FlipWidthScaling=False
    OptionalParentItem=(Type=Bone,Name="None")
    RotationOffset=(X=90,Y=0,Z=-55)
    ExecutePin=@RigVMFunction_Sequence.A
   )
   n2 = RigVMVariableNode() {variable=isTargeting}
   n3 = RigVMDispatchNode(
    False=(Rotation=(X=0.1464,Y=-0.3536,Z=0.3536,W=0.8536),Translation=(X=-10,Y=0,Z=0),Scale3D=(X=1,Y=1,Z=1))
    True=(Rotation=(X=0,Y=-0.3827,Z=0,W=0.9239),Translation=(X=0,Y=0,Z=0),Scale3D=(X=1,Y=1,Z=1))
    False.Translation.X=RigVMDispatchNode(
     True=-5
     False=RigVMDispatchNode(False=0, True=-7, Condition=@n2.Value) {template="DISPATCH_RigVMDispatch_If(in Condition,in True,in False,out Result)"}.Result
     Condition=RigVMVariableNode() {variable=PlayEmote}.Value
    ) {template="DISPATCH_RigVMDispatch_If(in Condition,in True,in False,out Result)"}.Result
    Condition=RigVMVariableNode() {variable=skydiving}.Value
   ) {template="DISPATCH_RigVMDispatch_If(in Condition,in True,in False,out Result)"}
   > &RigUnit_ModifyTransforms(
    ItemToModify="((Item=(Type=Bone,Name=\"base_hair_long_bk_c\"),Transform=(Rotation=(X=0,Y=-0.3827,Z=0,W=0.9239),Translation=(X=0,Y=0,Z=0),Scale3D=(X=1,Y=1,Z=1))),(Item=(Type=Bone,Name=\"base_hair_long_bk_l\"),Transform=(Rotation=(X=0,Y=-0.3827,Z=0,W=0.9239),Translation=(X=0, ...(truncated)"
    WeightMinimum=0
    ItemToModify.4.Transform=@n3.Result
    Weight=@n1.OutputParam
   )
  ]}
]
hierarchy [
 root {loc=<0 0 0>, rot=<0 0 0 1>}
 pelvis {parent=root, loc=<0 -0.5301 98.0571>, rot=<0.0127 0.707 -0.0127 -0.707>}
 spine_01 {parent=pelvis, loc=<6.1208 -0.9328 0>, rot=<0 0 0.034 -0.9994>}
 spine_02 {parent=spine_01, loc=<6.5997 -0.2095 0>, rot=<0 0 0.0238 -0.9997>}
 spine_03 {parent=spine_02, loc=<6.4914 -0.1237 0>, rot=<0 0 0.0149 -0.9999>}
 spine_04 {parent=spine_03, loc=<6.9598 0.0794 0>, rot=<0 0 -0.0132 -0.9999>}
 spine_05 {parent=spine_04, loc=<6.9987 0 0>, rot=<0 0 -0.0015 -1>}
 neck_01 {parent=spine_05, loc=<11.3138 0.1529 0>, rot=<0 0 -0.2003 -0.9797>}
 neck_02 {parent=neck_01, loc=<4.5627 0 0>, rot=<0 0 0.0504 -0.9987>}
 head {parent=neck_02, loc=<5.3914 0 0>, rot=<0 0 0.115 -0.9934>}
 base_hair_long_side_r {parent=head, loc=<2.8612 0.325 7.5248>, rot=<-0.0257 -0.7057 0.7074 0.0309>}
 dyn_hair_long_side_r_01 {parent=base_hair_long_side_r, loc=<0 0 0>, rot=<0 0 0 1>}
 dyn_hair_long_side_r_02 {parent=dyn_hair_long_side_r_01, loc=<4.6547 0 0>, rot=<0 0 0 1>}
 dyn_hair_long_side_r_03 {parent=dyn_hair_long_side_r_02, loc=<4.2033 -0.0001 0.0002>, rot=<0 0 0 1>}
 dyn_hair_long_side_r_04 {parent=dyn_hair_long_side_r_03, loc=<4.6747 0 -0.0001>, rot=<0 0 0 1>}
 base_hair_long_side_l {parent=head, loc=<2.8617 0.3251 -7.5248>, rot=<-0.0355 -0.7218 0.6899 -0.0411>}
 dyn_hair_long_side_l_01 {parent=base_hair_long_side_l, loc=<0 0 0>, rot=<0 0 0 1>}
 dyn_hair_long_side_l_02 {parent=dyn_hair_long_side_l_01, loc=<4.655 0 0>, rot=<0 0 0 1>}
 dyn_hair_long_side_l_03 {parent=dyn_hair_long_side_l_02, loc=<4.2036 0 0.0002>, rot=<0 0 0 1>}
 dyn_hair_long_side_l_04 {parent=dyn_hair_long_side_l_03, loc=<4.6744 0 0>, rot=<0 0 0 1>}
 base_hair_long_side_fr_l {parent=head, loc=<6.0013 4.5175 -6.2013>, rot=<-0.0427 -0.7748 0.609 -0.1642>}
 dyn_hair_long_side_fr_l_01 {parent=base_hair_long_side_fr_l, loc=<0 0 0>, rot=<0 0 0 1>}
 dyn_hair_long_side_fr_l_02 {parent=dyn_hair_long_side_fr_l_01, loc=<6.8359 0.2648 0>, rot=<-0.0057 0.0507 -0.1114 -0.9925>}
 dyn_hair_long_side_fr_l_03 {parent=dyn_hair_long_side_fr_l_02, loc=<6.5572 0 0.0002>, rot=<0 0 0 1>}
 base_hair_long_side_fr_r {parent=head, loc=<7.4044 5.4364 5.4818>, rot=<0.1309 -0.6724 0.7271 0.046>}
 dyn_hair_long_side_fr_r_01 {parent=base_hair_long_side_fr_r, loc=<0 0 0>, rot=<0 0 0 1>}
 dyn_hair_long_side_fr_r_02 {parent=dyn_hair_long_side_fr_r_01, loc=<5.1099 -0.0549 0.0676>, rot=<-0.0034 0.0898 0.0592 -0.9942>}
 dyn_hair_long_side_fr_r_03 {parent=dyn_hair_long_side_fr_r_02, loc=<5.1568 -0.0317 0.3884>, rot=<-0.003 0.1921 0.0151 -0.9813>}
 base_hair_long_l {parent=head, loc=<2.4776 -6.4398 -5.4982>, rot=<0.0316 -0.7064 0.7064 -0.0316>}
 dyn_hair_long_l_01 {parent=base_hair_long_l, loc=<0 0 0>, rot=<0 0 0 1>}
 dyn_hair_long_l_02 {parent=dyn_hair_long_l_01, loc=<6.347 0 0>, rot=<0 0 0 1>}
 dyn_hair_long_l_03 {parent=dyn_hair_long_l_02, loc=<5.8683 0 0.0002>, rot=<0 0 0 1>}
 dyn_hair_long_l_04 {parent=dyn_hair_long_l_03, loc=<5.3356 0 0.0689>, rot=<0 -0.0641 0 -0.9979>}
 base_hair_long_c {parent=head, loc=<3.2243 -7.1909 0.3046>, rot=<0.0316 -0.7064 0.7064 -0.0316>}
 dyn_hair_long_c_01 {parent=base_hair_long_c, loc=<0 0 0>, rot=<0 0 0 1>}
 dyn_hair_long_c_02 {parent=dyn_hair_long_c_01, loc=<6.3469 0 0>, rot=<0 0 0 1>}
 dyn_hair_long_c_03 {parent=dyn_hair_long_c_02, loc=<7.5222 0 0>, rot=<0 -0.1129 0 -0.9936>}
 dyn_hair_long_c_04 {parent=dyn_hair_long_c_03, loc=<6.1272 0 0.2297>, rot=<0 -0.0094 0 -1>}
 base_hair_long_r {parent=head, loc=<2.6046 -5.7892 5.8191>, rot=<0.0316 -0.7064 0.7064 -0.0316>}
 dyn_hair_long_r_01 {parent=base_hair_long_r, loc=<0 0 0>, rot=<0 0 0 1>}
 dyn_hair_long_r_02 {parent=dyn_hair_long_r_01, loc=<6.3469 0 0>, rot=<0 0 0 1>}
 dyn_hair_long_r_03 {parent=dyn_hair_long_r_02, loc=<5.8685 0 0>, rot=<0 0 0 1>}
 dyn_hair_long_r_04 {parent=dyn_hair_long_r_03, loc=<5.3359 0 0.0688>, rot=<0 -0.0641 0 -0.9979>}
 base_hair_long_bk_l {parent=head, loc=<7.463 -5.6091 -5.4982>, rot=<0.093 -0.701 0.701 -0.093>}
 dyn_hair_long_bk_l_01 {parent=base_hair_long_bk_l, loc=<0 0 0>, rot=<0 0 0 1>}
 dyn_hair_long_bk_l_02 {parent=dyn_hair_long_bk_l_01, loc=<6.347 0 0>, rot=<0 0 0 1>}
 dyn_hair_long_bk_l_03 {parent=dyn_hair_long_bk_l_02, loc=<5.8683 0 0.0002>, rot=<0 0 0 1>}
 dyn_hair_long_bk_l_04 {parent=dyn_hair_long_bk_l_03, loc=<5.3356 0 0.0689>, rot=<0 -0.0641 0 -0.9979>}
 base_hair_long_bk_c {parent=head, loc=<8.2097 -6.3603 0.3046>, rot=<0.093 -0.701 0.701 -0.093>}
 dyn_hair_long_bk_c_01 {parent=base_hair_long_bk_c, loc=<0 0 0>, rot=<0 0 0 1>}
 dyn_hair_long_bk_c_02 {parent=dyn_hair_long_bk_c_01, loc=<6.3469 0 0>, rot=<0 0 0 1>}
 dyn_hair_long_bk_c_03 {parent=dyn_hair_long_bk_c_02, loc=<7.5222 0 0>, rot=<0 -0.0054 0 -1>}
 dyn_hair_long_bk_c_04 {parent=dyn_hair_long_bk_c_03, loc=<6.1297 0 -0.1487>, rot=<0 0 0 1>}
 base_hair_long_bk_r {parent=head, loc=<7.59 -4.9585 5.8191>, rot=<0.093 -0.701 0.701 -0.093>}
 dyn_hair_long_bk_r_01 {parent=base_hair_long_bk_r, loc=<0 0 0>, rot=<0 0 0 1>}
 dyn_hair_long_bk_r_02 {parent=dyn_hair_long_bk_r_01, loc=<6.3469 0 0>, rot=<0 0 0 1>}
 dyn_hair_long_bk_r_03 {parent=dyn_hair_long_bk_r_02, loc=<5.8685 0 0>, rot=<0 0 0 1>}
 dyn_hair_long_bk_r_04 {parent=dyn_hair_long_bk_r_03, loc=<5.3359 0 0.0688>, rot=<0 -0.0641 0 -0.9979>}
 clavicle_l {parent=spine_05, loc=<5.4831 -0.1053 -3.2946>, rot=<0.0465 -0.7289 -0.009 -0.6829>}
 upperarm_l {parent=clavicle_l, loc=<9.8658 -1.2136 -0.4105>, rot=<0.0164 -0.3245 -0.0269 -0.9454>}
 clavicle_r {parent=spine_05, loc=<5.4831 -0.1053 3.2946>, rot=<0.7289 0.0465 -0.6829 0.009>}
 upperarm_r {parent=clavicle_r, loc=<-9.8658 1.2136 0.4105>, rot=<0.0164 -0.3245 -0.0269 -0.9454>}
]
variables [
 PlayEmote {type=bool, default=False}
 isTargeting {type=bool, default=False}
 skydiving {type=bool, default=False}
]
