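# Physics asset for the TangyRadishFlameLong female medium character's face accessory mesh, setting up physics for long hair.
# Seven capsule chains (three back, two side, two front-side) hang from base bones constrained to the head, with swing limits widening toward the back tips.
# Free-rotation cross constraints link neighboring lower strands, tying most of the hair together laterally.
# Unconstrained capsules on head, neck, upper spine, clavicles and upper arms supply body collision shapes.
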PhysicsAsset /BRCosmetics/Characters/Player/Female/Medium/Bodies/F_MED_TangyRadishFlameLong/Meshes/Parts/F_MED_TangyRadishFlameLong_FaceAcc_Physics
bodies {
 base_hair_long_bk_c [{shapeType=Capsule, length=1, radius=1, shapeName=base_hair_long_bk_c_capsule, center=<1 0 0>, rotation=<rot -90 0 0>}]
 base_hair_long_bk_l [{shapeType=Capsule, length=1, radius=1, shapeName=base_hair_long_bk_l_capsule, center=<1 0 0>, rotation=<rot -90 0 0>}]
 base_hair_long_bk_r [{shapeType=Capsule, length=1, radius=1, shapeName=base_hair_long_bk_r_capsule, center=<1 0 0>, rotation=<rot -90 0 0>}]
 base_hair_long_side_fr_l [{shapeType=Capsule, length=1, radius=1, shapeName=base_hair_long_side_fr_l_capsule, center=<1 0 0>, rotation=<rot -90 0 0>}]
 base_hair_long_side_fr_r [{shapeType=Capsule, length=1, radius=1, shapeName=base_hair_long_side_fr_r_capsule, center=<1 0 0>, rotation=<rot -90 0 0>}]
 base_hair_long_side_l [{shapeType=Capsule, length=1, radius=1, shapeName=base_hair_long_side_l_capsule, center=<1 0 0>, rotation=<rot -90 0 0>}]
 base_hair_long_side_r [{shapeType=Capsule, length=1, radius=1, shapeName=base_hair_long_side_r_capsule, center=<1 0 0>, rotation=<rot -90 0 0>}]
 clavicle_l [{shapeType=Capsule, length=5.8423, radius=5.8423, shapeName=clavicle_l_capsule, center=<4.6227 0 -1.5>, rotation=<rot -90 0 0>}]
 clavicle_r [{shapeType=Capsule, length=5.8423, radius=6, shapeName=clavicle_r_capsule, center=<-4.6227 0 1.5>, rotation=<rot -90 0 0>}]
 dyn_hair_long_bk_c_01 [{shapeType=Capsule, length=3, radius=2, center=<10 0 0>, rotation=<rot -90 0 0>}]
 dyn_hair_long_bk_c_02 [{shapeType=Capsule, length=2.4499, radius=3.5501, center=<10.567 0 0>, rotation=<rot -90 90 0>}]
 dyn_hair_long_bk_c_03 [{shapeType=Capsule, length=2.6004, radius=3.3996, center=<9.8595 0 0>, rotation=<rot -90 90 0>}]
 dyn_hair_long_bk_c_04 [{shapeType=Capsule, length=3.5, radius=3, center=<10 0 0>, rotation=<rot -90 90 0>}]
 dyn_hair_long_bk_l_01 [{shapeType=Capsule, length=3, radius=2, center=<10 0 0>, rotation=<rot -90 0 0>}]
 dyn_hair_long_bk_l_02 [{shapeType=Capsule, length=3, radius=2.5, center=<10 0 0>, rotation=<rot -90 90 0>}]
 dyn_hair_long_bk_l_03 [{shapeType=Capsule, length=3, radius=2.5, center=<10 0 0>, rotation=<rot -90 90 0>}]
 dyn_hair_long_bk_l_04 [{shapeType=Capsule, length=4, radius=2.2, center=<10 0 0>, rotation=<rot -90 90 0>}]
 dyn_hair_long_bk_r_01 [{shapeType=Capsule, length=3, radius=2, center=<10 0 0>, rotation=<rot -90 0 0>}]
 dyn_hair_long_bk_r_02 [{shapeType=Capsule, length=5.5505, radius=2.9495, center=<10 0 0>, rotation=<rot -90 90 0>}]
 dyn_hair_long_bk_r_03 [{shapeType=Capsule, length=5.7384, radius=3, center=<10 0 0>, rotation=<rot -90 90 0>}]
 dyn_hair_long_bk_r_04 [{shapeType=Capsule, length=4, radius=2.2, center=<10 0 0>, rotation=<rot -90 90 0>}]
 dyn_hair_long_side_fr_l_01 [{shapeType=Capsule, length=3, radius=1.5, center=<3 0 0>, rotation=<rot -90 0 0>}]
 dyn_hair_long_side_fr_l_02 [{shapeType=Capsule, length=3, radius=1.5, center=<3 0 0>, rotation=<rot -90 0 0>}]
 dyn_hair_long_side_fr_l_03 [{shapeType=Capsule, length=3, radius=2.5, center=<3 0 0>, rotation=<rot -90 0 0>}]
 dyn_hair_long_side_fr_r_01 [{shapeType=Capsule, length=3, radius=1.5, center=<1 0 0>, rotation=<rot -90 0 0>}]
 dyn_hair_long_side_fr_r_02 [{shapeType=Capsule, length=3, radius=1.5, center=<1 0 0>, rotation=<rot -90 0 0>}]
 dyn_hair_long_side_fr_r_03 [{shapeType=Capsule, length=3, radius=1.5, center=<1 0 0>, rotation=<rot -90 0 0>}]
 dyn_hair_long_side_l_01 [{shapeType=Capsule, length=3, radius=2, center=<-2 0 0>, rotation=<rot 0 90 0>}]
 dyn_hair_long_side_l_02 [{shapeType=Capsule, length=3, radius=2.2, center=<1 0 0>, rotation=<rot 0 90 0>}]
 dyn_hair_long_side_l_03 [{shapeType=Capsule, length=3, radius=2.2, center=<1 0 0>, rotation=<rot 0 90 0>}]
 dyn_hair_long_side_l_04 [{shapeType=Capsule, length=3, radius=2.2, center=<-2 0 0>, rotation=<rot 0 90 0>}]
 dyn_hair_long_side_r_01 [{shapeType=Capsule, length=3, radius=2, center=<-2 0 0>, rotation=<rot 0 90 0>}]
 dyn_hair_long_side_r_02 [{shapeType=Capsule, length=3, radius=2.2, center=<1 0 0>, rotation=<rot 0 90 0>}]
 dyn_hair_long_side_r_03 [{shapeType=Capsule, length=3, radius=2.2, center=<1 0 0>, rotation=<rot 0 90 0>}]
 dyn_hair_long_side_r_04 [{shapeType=Capsule, length=6, radius=3, center=<2 -1 0>, rotation=<rot 0 90 0>}]
 head [{shapeType=Capsule, length=6.1281, radius=6.1281, shapeName=head_capsule, center=<0.2517 1.1618 0>, rotation=<rot -90 0 0>}]
 neck_01 [{shapeType=Capsule, length=5.2895, radius=5.2895, shapeName=neck_01_capsule, center=<0.2517 1.1618 0>, rotation=<rot -90 0 0>}]
 neck_02 [{shapeType=Capsule, length=5.2895, radius=5.2895, shapeName=neck_02_capsule, center=<0.2517 1.1618 0>, rotation=<rot -90 0 0>}]
 spine_03 [{shapeType=Capsule, length=7.1034, radius=7.1132, shapeName=spine_03_capsule}]
 spine_04 [{shapeType=Capsule, length=8.0615, radius=8.0615, shapeName=spine_04_capsule}]
 spine_05 [{shapeType=Capsule, length=8.7727, radius=8.7727, shapeName=spine_05_capsule, center=<0 0.4908 0>}]
 upperarm_l [{shapeType=Capsule, length=25, radius=5.2068, shapeName=upperarm_l_capsule, center=<12.5 -0.0736 -0.7766>, rotation=<rot -90 0 0>}]
 upperarm_r [{shapeType=Capsule, length=25, radius=5.5, shapeName=upperarm_r_capsule, center=<-12.5 0.0736 1.5>, rotation=<rot -90 0 0>}]
}
constraints [
 {
  bone1Name=base_hair_long_bk_c
  bone2Name=head
  swing1LimitDegrees=45
  swing1Motion=Free
  swing2LimitDegrees=45
  swing2Motion=Free
  twistLimitDegrees=45
  twistMotion=Free
 }
 {bone1Name=dyn_hair_long_bk_c_01, bone2Name=base_hair_long_bk_c, swing1LimitDegrees=5, swing2LimitDegrees=10, twistLimitDegrees=2}
 {bone1Name=dyn_hair_long_bk_c_02, bone2Name=dyn_hair_long_bk_c_01, swing1LimitDegrees=10, swing2LimitDegrees=15, twistLimitDegrees=2}
 {bone1Name=dyn_hair_long_bk_c_03, bone2Name=dyn_hair_long_bk_c_02, swing1LimitDegrees=15, swing2LimitDegrees=25, twistLimitDegrees=2}
 {
  bone1Name=dyn_hair_long_bk_l_03
  bone2Name=dyn_hair_long_bk_c_03
  swing1LimitDegrees=45
  swing1Motion=Free
  swing2LimitDegrees=45
  swing2Motion=Free
  twistLimitDegrees=45
  twistMotion=Free
 }
 {
  bone1Name=dyn_hair_long_bk_r_03
  bone2Name=dyn_hair_long_bk_c_03
  swing1LimitDegrees=45
  swing1Motion=Free
  swing2LimitDegrees=45
  swing2Motion=Free
  twistLimitDegrees=45
  twistMotion=Free
 }
 {bone1Name=dyn_hair_long_bk_c_04, bone2Name=dyn_hair_long_bk_c_03, swing1LimitDegrees=25, swing2LimitDegrees=45, twistLimitDegrees=2}
 {
  bone1Name=dyn_hair_long_bk_l_04
  bone2Name=dyn_hair_long_bk_c_04
  swing1LimitDegrees=45
  swing1Motion=Free
  swing2LimitDegrees=45
  swing2Motion=Free
  twistLimitDegrees=45
  twistMotion=Free
 }
 {
  bone1Name=dyn_hair_long_bk_r_04
  bone2Name=dyn_hair_long_bk_c_04
  swing1LimitDegrees=45
  swing1Motion=Free
  swing2LimitDegrees=45
  swing2Motion=Free
  twistLimitDegrees=45
  twistMotion=Free
 }
 {
  bone1Name=base_hair_long_bk_l
  bone2Name=head
  swing1LimitDegrees=45
  swing1Motion=Free
  swing2LimitDegrees=45
  swing2Motion=Free
  twistLimitDegrees=45
  twistMotion=Free
 }
 {bone1Name=dyn_hair_long_bk_l_01, bone2Name=base_hair_long_bk_l, swing1LimitDegrees=5, swing2LimitDegrees=10, twistLimitDegrees=2}
 {bone1Name=dyn_hair_long_bk_l_02, bone2Name=dyn_hair_long_bk_l_01, swing1LimitDegrees=10, swing2LimitDegrees=15, twistLimitDegrees=2}
 {
  bone1Name=dyn_hair_long_side_fr_l_03
  bone2Name=dyn_hair_long_bk_l_02
  swing1LimitDegrees=45
  swing1Motion=Free
  swing2LimitDegrees=45
  swing2Motion=Free
  twistLimitDegrees=45
  twistMotion=Free
 }
 {bone1Name=dyn_hair_long_bk_l_03, bone2Name=dyn_hair_long_bk_l_02, swing1LimitDegrees=15, swing2LimitDegrees=25, twistLimitDegrees=2}
 {
  bone1Name=dyn_hair_long_side_l_03
  bone2Name=dyn_hair_long_bk_l_03
  swing1LimitDegrees=45
  swing1Motion=Free
  swing2LimitDegrees=45
  swing2Motion=Free
  twistLimitDegrees=45
  twistMotion=Free
 }
 {bone1Name=dyn_hair_long_bk_l_04, bone2Name=dyn_hair_long_bk_l_03, swing1LimitDegrees=25, swing2LimitDegrees=45, twistLimitDegrees=25}
 {
  bone1Name=dyn_hair_long_side_l_04
  bone2Name=dyn_hair_long_bk_l_04
  swing1LimitDegrees=45
  swing1Motion=Free
  swing2LimitDegrees=45
  swing2Motion=Free
  twistLimitDegrees=45
  twistMotion=Free
 }
 {
  bone1Name=base_hair_long_bk_r
  bone2Name=head
  swing1LimitDegrees=45
  swing1Motion=Free
  swing2LimitDegrees=45
  swing2Motion=Free
  twistLimitDegrees=45
  twistMotion=Free
 }
 {bone1Name=dyn_hair_long_bk_r_01, bone2Name=base_hair_long_bk_r, swing1LimitDegrees=5, swing2LimitDegrees=10, twistLimitDegrees=2}
 {bone1Name=dyn_hair_long_bk_r_02, bone2Name=dyn_hair_long_bk_r_01, swing1LimitDegrees=10, swing2LimitDegrees=15, twistLimitDegrees=2}
 {bone1Name=dyn_hair_long_bk_r_03, bone2Name=dyn_hair_long_bk_r_02, swing1LimitDegrees=15, swing2LimitDegrees=25, twistLimitDegrees=2}
 {
  bone1Name=dyn_hair_long_side_r_03
  bone2Name=dyn_hair_long_bk_r_03
  swing1LimitDegrees=45
  swing1Motion=Free
  swing2LimitDegrees=45
  swing2Motion=Free
  twistLimitDegrees=45
  twistMotion=Free
 }
 {bone1Name=dyn_hair_long_bk_r_04, bone2Name=dyn_hair_long_bk_r_03, swing1LimitDegrees=25, swing2LimitDegrees=45, twistLimitDegrees=25}
 {
  bone1Name=base_hair_long_side_fr_l
  bone2Name=head
  swing1LimitDegrees=45
  swing1Motion=Free
  swing2LimitDegrees=45
  swing2Motion=Free
  twistLimitDegrees=45
  twistMotion=Free
 }
 {bone1Name=dyn_hair_long_side_fr_l_01, bone2Name=base_hair_long_side_fr_l, swing1LimitDegrees=25, swing2LimitDegrees=25, twistLimitDegrees=2}
 {
  bone1Name=dyn_hair_long_side_fr_l_02
  bone2Name=dyn_hair_long_side_fr_l_01
  swing1LimitDegrees=25
  swing2LimitDegrees=25
  twistLimitDegrees=2
 }
 {
  bone1Name=dyn_hair_long_side_fr_l_03
  bone2Name=dyn_hair_long_side_fr_l_02
  swing1LimitDegrees=25
  swing2LimitDegrees=25
  twistLimitDegrees=2
 }
 {
  bone1Name=base_hair_long_side_fr_r
  bone2Name=head
  swing1LimitDegrees=45
  swing1Motion=Free
  swing2LimitDegrees=45
  swing2Motion=Free
  twistLimitDegrees=45
  twistMotion=Free
 }
 {
  bone1Name=dyn_hair_long_side_fr_r_01
  bone2Name=base_hair_long_side_fr_r
  swing1LimitDegrees=25
  swing2LimitDegrees=15
  twistLimitDegrees=2
  twistMotion=Locked
 }
 {
  bone1Name=dyn_hair_long_side_fr_r_02
  bone2Name=dyn_hair_long_side_fr_r_01
  swing1LimitDegrees=25
  swing2LimitDegrees=15
  twistLimitDegrees=2
  twistMotion=Locked
 }
 {
  bone1Name=dyn_hair_long_side_fr_r_03
  bone2Name=dyn_hair_long_side_fr_r_02
  swing1LimitDegrees=25
  swing2LimitDegrees=15
  twistLimitDegrees=2
  twistMotion=Locked
 }
 {
  bone1Name=base_hair_long_side_l
  bone2Name=head
  swing1LimitDegrees=45
  swing1Motion=Free
  swing2LimitDegrees=45
  swing2Motion=Free
  twistLimitDegrees=45
  twistMotion=Free
 }
 {bone1Name=dyn_hair_long_side_l_01, bone2Name=base_hair_long_side_l, swing1LimitDegrees=25, swing2LimitDegrees=22.5, twistLimitDegrees=2}
 {bone1Name=dyn_hair_long_side_l_02, bone2Name=dyn_hair_long_side_l_01, swing1LimitDegrees=30, swing2LimitDegrees=22.5, twistLimitDegrees=22.5}
 {bone1Name=dyn_hair_long_side_l_03, bone2Name=dyn_hair_long_side_l_02, swing1LimitDegrees=30, swing2LimitDegrees=22.5, twistLimitDegrees=22.5}
 {bone1Name=dyn_hair_long_side_l_04, bone2Name=dyn_hair_long_side_l_03, swing1LimitDegrees=25, swing2LimitDegrees=25, twistLimitDegrees=22.5}
 {
  bone1Name=base_hair_long_side_r
  bone2Name=head
  swing1LimitDegrees=45
  swing1Motion=Free
  swing2LimitDegrees=45
  swing2Motion=Free
  twistLimitDegrees=45
  twistMotion=Free
 }
 {bone1Name=dyn_hair_long_side_r_01, bone2Name=base_hair_long_side_r, swing1LimitDegrees=25, swing2LimitDegrees=25, twistLimitDegrees=2}
 {bone1Name=dyn_hair_long_side_r_02, bone2Name=dyn_hair_long_side_r_01, swing1LimitDegrees=25, swing2LimitDegrees=25, twistLimitDegrees=22.5}
 {bone1Name=dyn_hair_long_side_r_03, bone2Name=dyn_hair_long_side_r_02, swing1LimitDegrees=25, swing2LimitDegrees=25, twistLimitDegrees=22.5}
 {bone1Name=dyn_hair_long_side_r_04, bone2Name=dyn_hair_long_side_r_03, swing1LimitDegrees=25, swing2LimitDegrees=25, twistLimitDegrees=22.5}
]
